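# Niagara effect for the OboeThorn emote's prop spawn, combining a FlowersMin emitter with lightweight sparkle, airglow and smoke emitters.
# FlowersMin fires two instant bursts in separate spawn groups, writing the group to VisibilityTag for its mesh and sprite renderers.
# Short-lived, randomly colored particles launch from a shape volume, slow under drag, spin, shrink and fade.
# Time-of-day color inputs tint the flowers and petals.
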NiagaraSystem /BRCosmetics/Animation/Game/MainPlayer/Emotes/OboeThorn/FX/NS_Emote_OboeThorn_PropSpawn
dependencies {
 usedDataInterfaces [&NiagaraDataInterfaceArrayColor, &NiagaraDataInterfaceCurve]
 usedDynamicInputs []
 usedModules [
  &self:SystemSpawnScript.SystemScriptSource.SystemScriptGraph.NiagaraNodeAssignment_0.SetVariables_#1
  &SystemState
 ]
}
emitters {
 AirglowLWE {rendererClasses [], renderers []}
 FlowersMin {
  emitterSpawnScript [/SetVariables(
    Emitter.Color1:LinearColor=&FNTimeOfDayColor(
     Night Color=<rgba 0.2917 0.0688 0.0395 1>
     Morning Color=<rgba 1 0.1417 0.0607 1>
     Day Color=<rgba 1 0.1417 0.0607 1>
     Evening Color=<rgba 1 0.1417 0.0607 1>
    )
    Emitter.Color2:LinearColor=&FNTimeOfDayColor(
     Night Color=<rgba 0.3333 0.2961 0.0579 1>
     Morning Color=<rgba 1 0.8818 0.0868 1>
     Day Color=<rgba 1 0.8818 0.0868 1>
     Evening Color=<rgba 1 0.8818 0.0868 1>
    )
   ) {isSetParameters=true}]
  emitterUpdateScript [
   &EmitterState(Inactive Response=::NewEnumerator1, Life Cycle Mode=::NewEnumerator1, Loop Behavior=::NewEnumerator1, Loop Duration=3, UseLoopDelay=false)
   &SpawnBurst_Instantaneous(Spawn Count=10, Spawn Group=1, Use Spawn Probability=false, UseLoopCountLimit=false) {module=SpawnBurst_Instantaneous001}
   &SpawnBurst_Instantaneous(Spawn Count=14, Use Spawn Probability=true, UseLoopCountLimit=false)
  ]
  particleSpawnScript [
   &InitializeParticle(
    Color=&SelectLinearColorFromArray(
     Array Sampling Mode=::NewEnumerator0
     Direct Array Index=?
     Color Selection Array=DataInterface() {propertyValues {ColorData [<rgba 0 0.4541 1 1>, <rgba 1 0.2691 0.625 1>, <rgba 1 0.9295 0.1684 1>, <rgba 0.9841 1 0.8498 1>]}}
     Random Seed=0
    )
    Lifetime Max=0.7
    Lifetime Min=0.4
    Lifetime Mode=::NewEnumerator1
    Mesh Scale Mode=::NewEnumerator3
    Mesh Uniform Scale=&Multiply_Float(A=&RandomRangeFloat(Minimum=0.2, Maximum=0.7), B=0.6)
    Sprite Rotation Mode=::NewEnumerator2
    Sprite Size Mode=::NewEnumerator3
    UsePositionOffset=false
    Write Material Random=true
   )
   /SetVariables(Particles.PetalLifetime:NiagaraFloat=&RandomRangeFloat(Minimum=0.4, Maximum=1)) {isSetParameters=true}
   &ShapeLocation(
    Box Size=<40 70 50>
    MaskSpawnGroup=false
    Offset Mode=::NewEnumerator0
    Override Seed=false
    Shape Primitive=::NewEnumerator2
    Surface Only=false
    UseOwnerScale=true
   ) {module=ShapeLocation001}
   &AddVelocity(
    Module.Use Falloff Distance=false
    Origin Offset=<0 0 -25>
    Override Seed=false
    Velocity Mode=::NewEnumerator1
    Velocity Speed=&RandomRangeFloat(Minimum=350, Maximum=435)
   )
   &CameraOffset(Camera Offset Amount=20)
   &InitialMeshOrientation(Mesh Axis To Orient=<-1 0 0>, Mesh Orientation Mode=::NewEnumerator2, Orientation Vector=<0 0 1>, Use Rotation Vector=false)
   /SetVariables(Particles.VisibilityTag=&SpawnGroup()) {isSetParameters=true}
   /SetVariables(Particles.Petal Color:LinearColor=&FNTimeOfDayColor(
     Night Color=<rgba 0.5833 0.1031 0.0577 1>
     Morning Color=<rgba 0.5781 0.0819 0.0351 1>
     Day Color=<rgba 0.8385 0.1481 0.083 1>
     Evening Color=<rgba 0.5781 0.0819 0.0351 1>
    )) {isSetParameters=true}
  ]
  particleUpdateScript [
   &ParticleState()
   &ScaleSpriteSize(Uniform Curve Sprite Scale=DataInterface() {propertyValues {Curve {keys [
        {tangentMode=RCTM_User, time=0, value=1}
        {arriveTangentWeight=0.2822, interpMode=RCIM_Cubic, leaveTangentWeight=0.1893, tangentMode=RCTM_User, time=0.5985, value=1}
        {arriveTangent=-2.7277, interpMode=RCIM_Cubic, leaveTangent=-2.7277, tangentMode=RCTM_User, time=1, value=0}
       ]}}})
   &ScaleMeshSize(Scale Factor=&VectorFromFloat(Value=&FloatFromCurve(
      FloatCurve=DataInterface() {propertyValues {Curve {keys [
          {interpMode=RCIM_Cubic, tangentMode=RCTM_User, time=0, value=0}
          <at 0.2209 1>
          {interpMode=RCIM_Cubic, tangentMode=RCTM_User, time=0.7957, value=1}
          {tangentMode=RCTM_User, time=1, value=0}
         ]}}}
      Scale Curve=&Multiply_FloatByInt(Float=1, Integer=&RandomRangeInt(Minimum=0, Maximum=1))
     )))
   &UpdateMeshOrientation(
    Rotation Rate=&FloatFromCurve(
     FloatCurve=DataInterface() {propertyValues {Curve {keys [
         {interpMode=RCIM_Cubic, tangentMode=RCTM_User, time=0, value=0}
         <at 0.1684 1>
         {interpMode=RCIM_Cubic, tangentMode=RCTM_User, time=0.4961, value=1}
         {tangentMode=RCTM_User, time=1, value=0}
        ]}}}
     Scale Curve=&Multiply_FloatByInt(Float=&RandomRangeFloat(Minimum=-2, Maximum=2), Integer=1)
    )
    Rotation Vector=<0 0 1>
   )
   &ScaleColor(
    Scale Alpha=&FloatFromCurve(FloatCurve=DataInterface() {propertyValues {Curve {keys [<at -0.0046 1.0113 RCIM_Cubic>,<at 1.0011 0.002 RCIM_Cubic>]}}})
    Scale RGB=&MakeVectorFromLinearColorRGB(Color=&FNTimeOfDayColor(
      Night Color=<rgba 0.4102 0.5995 0.75 1>
      Morning Color=<rgba 1 0.6765 0.6094 1>
      Day Color=<rgba 0.7865 0.7865 0.7865 1>
      Evening Color=<rgba 0.7708 0.7708 0.7708 1>
     ))
    ScaleA=true
    ScaleRGB=true
   )
   &Drag(Drag=5, Rotational Drag=2, Use Linear Drag=true, Use Rotational Drag=true)
   &SolveForcesAndVelocity(Clamp Velocity=false, Limit Acceleration=false)
   &AlignSpriteToMeshOrientation(Mesh Orientation Relative Sprite Alignment Vector=<1 20 1>, SetAlignment=true, SetFacing=true)
  ]
  rendererClasses [&NiagaraMeshRendererProperties, &NiagaraSpriteRendererProperties]
  renderers [
   {rendererClass=&NiagaraMeshRendererProperties, rendererIndex=0}
   {rendererClass=&NiagaraSpriteRendererProperties, rendererIndex=1}
  ]
 }
 SmokeLWE {rendererClasses [], renderers []}
 SparklesLWE {rendererClasses [], renderers []}
}
summary {
 systemName=NS_Emote_OboeThorn_PropSpawn
 emitters [
  {emitterName=SparklesLWE, rendererClasses []}
  {emitterName=AirglowLWE, rendererClasses []}
  {emitterName=SmokeLWE, rendererClasses []}
  {emitterName=FlowersMin, rendererClasses [&NiagaraMeshRendererProperties, &NiagaraSpriteRendererProperties]}
 ]
 userVariables [User.NormalizedNotifyProgress {type=NiagaraFloat, defaultValue=0, description=""}]
}
